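# Sound cue for the ambient energy loop of the UFO variant of the directional launcher launch pad.
# A single looping wave player plays the energy loop wave at low volume and slightly lowered pitch.
# It borrows the base directional launcher's ambient attenuation, routes to the Creative gadgets sound class, and overrides concurrency to block new instances rather than stop playing ones.
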SoundCue /Game/Athena/Sounds/LaunchPad/DirectionalLauncher_UFO/DirectionalLauncher_UFO_Amb_Energy_Loop_Cue
root=WavePlayer_1
nodes [{
  id=WavePlayer_1
  properties {bLooping=true, soundWaveAssetPtr=&/DirectionalLauncher_UFO_Amb_Energy_Loop}
 }]
properties {
 attenuationSettings=/Game/Athena/Sounds/LaunchPad/DirectionalLauncher/DirectionalLauncher_Ambient_Att
 bOverrideConcurrency=true
 maxDistance=2700
 pitchMultiplier=0.95
 soundClassObject=&Fort_Gadgets_Creative
 volumeMultiplier=0.175
 concurrencyOverrides {resolutionRule=PreventNew}
}
